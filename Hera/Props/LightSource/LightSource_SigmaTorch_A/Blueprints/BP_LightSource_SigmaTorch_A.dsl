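; Sigma torch light source prop Blueprint that tints its fire effects and point light at construction; its event handlers are empty stubs.
; The construction script randomly jitters the flipbook flame sideways and converts FX Color's hue offset from orange into a hue-shift parameter for flame, sparks and outer fire materials.
; It sets the light from a separate color variable, then replaces the outer fire material with a shrine brazier instance colored by scaled FX Color.
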(event EventBeginPlay
  (|CallParentFunction) ; This node is disabled and will not be called.\nDrag off pins to build functionality.
)

(event Collision|EventActorBeginOverlap
  (params (OtherActor :type Actor%20Object%20Reference))
  (|CallParentFunction OtherActor) ; This node is disabled and will not be called.\nDrag off pins to build functionality.
)

(event EventTick
  (params (DeltaSeconds :type Float%20%28single-precision%29))
  (|CallParentFunction DeltaSeconds) ; This node is disabled and will not be called.\nDrag off pins to build functionality.
)

(fn ConstructionScript
  (bind _returnvalue (Math|Color|RGBtoHSV(LinearColor) "(R=1.000000,G=0.250000,B=0.000000,A=1.000000)"))
  (bind _fx_color (Variables|Default|GetFX%20Color))
  (bind _returnvalue_2 (Math|Color|RGBtoHSV(LinearColor) _fx_color))
  (bind _returnvalue_3 (- _returnvalue_2 _returnvalue))
  (bind _r (Utilities|Struct|BreakLinearColor _returnvalue_3))
  (bind _returnvalue_4 (/ _r 3.6e+02))
  (bind _fx_color_2 (Variables|Default|GetFX%20Color))
  (|CallParentFunction)
  (Transformation|SetRelativeLocation (Variables|Default|GetSM_Flame_Flipbook) (Math|Vector|MakeVector 0.0 (Math|Random|RandomFloatinRange -5.0 5.0) 2.3e+02) false false)
  (bind _returnvalue_5 (Rendering|Material|CreateDynamicMaterialInstance (Variables|Default|GetSM_Flame_Flipbook) 0 (Rendering|Material|GetMaterial (Variables|Default|GetSM_Flame_Flipbook) 0) "None" :!class /Script/Engine.PrimitiveComponent))
  (Rendering|Material|SetScalarParameterValue _returnvalue_5 "Hue Shift Percentage (S)" _returnvalue_4 :!class /Script/Engine.MaterialInstanceDynamic)
  (bind _returnvalue_6 (Rendering|Material|CreateDynamicMaterialInstance (Variables|Default|GetSparks_WPO_Mesh) 0 (Rendering|Material|GetMaterial (Variables|Default|GetSparks_WPO_Mesh) 0) "None" :!class /Script/Engine.PrimitiveComponent))
  (Rendering|Material|SetScalarParameterValue _returnvalue_6 "Hue Shift Percentage (S)" _returnvalue_4 :!class /Script/Engine.MaterialInstanceDynamic)
  (bind _returnvalue_7 (Rendering|Material|CreateDynamicMaterialInstance (Variables|Default|GetFire_Mesh_Outer) 0 (Rendering|Material|GetMaterial (Variables|Default|GetFire_Mesh_Outer) 0) "None" :!class /Script/Engine.PrimitiveComponent))
  (Rendering|Material|SetScalarParameterValue _returnvalue_7 "Hue Shift Percentage (S)" _returnvalue_4 :!class /Script/Engine.MaterialInstanceDynamic)
  (Rendering|Components|Light|SetLightColor (Variables|Default|GetPointLight) (Variables|Default|GetPointLight%20Color) true)
  (bind _returnvalue_8 (Rendering|Material|CreateDynamicMaterialInstance (Variables|Default|GetFire_Mesh_Outer) 0 "/Hera/FXEnvironment/Temple/Materials/MI_Env_Temple_ShrineBrazier_Fire.MI_Env_Temple_ShrineBrazier_Fire" "None" :!class /Script/Engine.PrimitiveComponent))
  (Rendering|Material|SetVectorParameterValue _returnvalue_8 "Color_A" (* _fx_color_2 "(R=7.000000,G=7.000000,B=7.000000,A=1.000000)") :!class /Script/Engine.MaterialInstanceDynamic)
  (Rendering|Material|SetVectorParameterValue _returnvalue_8 "Color_B" (* _fx_color_2 "(R=0.600000,G=0.600000,B=0.600000,A=1.000000)") :!class /Script/Engine.MaterialInstanceDynamic))
